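# Material function for Athena landscape materials that outputs a material attributes set with its base color replaced by a separate BaseColor input.
# It breaks the Mat_In input apart and rebuilds it with the new color.
# As wired here, every other connected attribute reads the Metallic output of the break node rather than its matching channel.
# Anisotropy, Tangent, ShadingModel and Displacement are left unconnected in the rebuilt set.
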MaterialFunction /FigmentEnvironment/FN_6_20/Athena/Environments/Landscape/MaterialFunctions/MF_Athena_ReplaceBaseColor
graph {nodes [
  n1 = BreakMaterialAttributes!(Attr=FunctionInput(Preview=MakeMaterialAttributes!(
     BaseColor
     Metallic
     Specular
     Roughness
     Anisotropy
     EmissiveColor
     Opacity
     OpacityMask
     Normal
     Tangent
     WorldPositionOffset
     SubsurfaceColor
     ClearCoat
     ClearCoatRoughness
     AmbientOcclusion
     Refraction
     CustomizedUVs_0
     CustomizedUVs_1
     CustomizedUVs_2
     CustomizedUVs_3
     CustomizedUVs_4
     CustomizedUVs_5
     CustomizedUVs_6
     CustomizedUVs_7
     PixelDepthOffset
     ShadingModel
     Displacement
    )) {inputName=Mat_In, inputType=FunctionInput_MaterialAttributes})
  FunctionOutput(None=MakeMaterialAttributes(
    BaseColor=FunctionInput() {inputName=BaseColor, sortPriority=1}
    Metallic=@n1.Metallic
    Specular=@n1.Metallic
    Roughness=@n1.Metallic
    EmissiveColor=@n1.Metallic
    Opacity=@n1.Metallic
    OpacityMask=@n1.Metallic
    Normal=@n1.Metallic
    WorldPositionOffset=@n1.Metallic
    SubsurfaceColor=@n1.Metallic
    ClearCoat=@n1.Metallic
    ClearCoatRoughness=@n1.Metallic
    AmbientOcclusion=@n1.Metallic
    Refraction=@n1.Metallic
    CustomizedUVs_0=@n1.Metallic
    CustomizedUVs_1=@n1.Metallic
    CustomizedUVs_2=@n1.Metallic
    CustomizedUVs_3=@n1.Metallic
    CustomizedUVs_4=@n1.Metallic
    CustomizedUVs_5=@n1.Metallic
    CustomizedUVs_6=@n1.Metallic
    CustomizedUVs_7=@n1.Metallic
    PixelDepthOffset=@n1.Metallic
   ))
 ]}
